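# Physics asset for the male medium ClumsyChew suit body, setting up secondary motion for the jacket's lower panels.
# Seven two-box chains (back, sides, fronts) hang from small MSTR anchor spheres on swing/twist-limited constraints.
# Mostly angular-free cross constraints link neighbouring lower boxes from front through back, so the panels move together.
# Pelvis, spine and thigh capsules supply body shapes; a long dyn_sword capsule has no constraint here.
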PhysicsAsset /BRCosmetics/Characters/Player/Male/Medium/Bodies/M_MED_ClumsyChew_Suit/Meshes/M_MED_ClumsyChew_Suit_Physics
bodies {
 MSTR_jacket_back [{shapeType=Sphere, radius=0.505, shapeName=MSTR_jacket_back_sphere}]
 MSTR_jacket_back_l [{shapeType=Sphere, radius=0.505, shapeName=MSTR_jacket_back_l_sphere}]
 MSTR_jacket_back_r [{shapeType=Sphere, radius=0.505, shapeName=MSTR_jacket_back_r_sphere}]
 MSTR_jacket_front_l [{shapeType=Sphere, radius=0.505, shapeName=MSTR_jacket_front_l_sphere}]
 MSTR_jacket_front_r [{shapeType=Sphere, radius=0.505, shapeName=MSTR_jacket_front_r_sphere}]
 MSTR_jacket_side_l [{shapeType=Sphere, radius=0.505, shapeName=MSTR_jacket_side_l_sphere}]
 MSTR_jacket_side_r [{shapeType=Sphere, radius=0.505, shapeName=MSTR_jacket_side_r_sphere}]
 dyn_jacket_back_01 [{
   shapeType=Box
   extentX=2.9144
   extentY=10.6973
   extentZ=13.0908
   shapeName=dyn_jacket_back_02_l_box
   center=<6.6947 -0.7643 -0.3503>
   rotation=<rot 0 90.9838 90.0001>
  }]
 dyn_jacket_back_01_l [{
   shapeType=Box
   extentX=3.1225
   extentY=9.1185
   extentZ=12.7322
   shapeName=dyn_jacket_back_02_l_box
   center=<6.5845 -0.0174 0.0239>
   rotation=<rot 10.4595 90.881 -89.8668>
  }]
 dyn_jacket_back_01_r [{
   shapeType=Box
   extentX=3.674
   extentY=9.7409
   extentZ=12.7322
   shapeName=dyn_jacket_back_02_r_box
   center=<6.5988 -0.3399 -0.478>
   rotation=<rot 15.8471 86.7752 88.911>
  }]
 dyn_jacket_back_02 [{
   shapeType=Box
   extentX=4.0399
   extentY=11.3174
   extentZ=13.4829
   shapeName=dyn_jacket_back_02_l_box
   center=<6.0438 0.0867 -0.3503>
   rotation=<rot 0 96.6472 90>
  }]
 dyn_jacket_back_02_l [{
   shapeType=Box
   extentX=2.9
   extentY=10.0057
   extentZ=13.4829
   shapeName=dyn_jacket_back_02_l_box
   center=<6.0807 0.1116 0.0479>
   rotation=<rot -4.3267 95.4484 -89.3205>
  }]
 dyn_jacket_back_02_r [{
   shapeType=Box
   extentX=2.9144
   extentY=10.0057
   extentZ=13.4829
   shapeName=dyn_jacket_back_02_r_box
   center=<6.089 -0.1059 -0.7943>
   rotation=<rot 12.2436 95.6468 90.6933>
  }]
 dyn_jacket_front_01_l [{
   shapeType=Box
   extentX=2.7108
   extentY=8.541
   extentZ=12.03
   shapeName=dyn_jacket_side_02_l_box
   center=<6.5502 -0.253 1.1057>
   rotation=<rot 1.9264 -86.9798 86.9862>
  }]
 dyn_jacket_front_01_r [{
   shapeType=Box
   extentX=2.5705
   extentY=6.3724
   extentZ=13.1087
   shapeName=dyn_jacket_side_02_r_box
   center=<6.5112 0.1336 -0.0104>
   rotation=<rot 5.544 -88.3903 -93.0499>
  }]
 dyn_jacket_front_02_l [{
   shapeType=Box
   extentX=2.2999
   extentY=8.7375
   extentZ=12.03
   shapeName=dyn_jacket_side_02_l_box
   center=<6.5107 -0.0457 1.6485>
   rotation=<rot -1.0832 -87.3264 87.0388>
  }]
 dyn_jacket_front_02_r [{
   shapeType=Box
   extentX=2.8262
   extentY=7.832
   extentZ=13.1087
   shapeName=dyn_jacket_side_02_r_box
   center=<6.5105 0.013 -0.0227>
   rotation=<rot -0.1913 -89.4949 -81.5959>
  }]
 dyn_jacket_side_01_l [{
   shapeType=Box
   extentX=3.4293
   extentY=9.4484
   extentZ=12.6554
   shapeName=dyn_jacket_side_02_l_box
   center=<7.0815 -0.4808 -0.191>
   rotation=<rot 12.5351 94.1324 -87.8769>
  }]
 dyn_jacket_side_01_r [{
   shapeType=Box
   extentX=3.3821
   extentY=8.1918
   extentZ=13.1087
   shapeName=dyn_jacket_side_02_r_box
   center=<6.5428 0.2103 0.773>
   rotation=<rot 5.5303 -88.6462 -84.4296>
  }]
 dyn_jacket_side_02_l [{
   shapeType=Box
   extentX=3.1082
   extentY=10.7097
   extentZ=12.5949
   shapeName=dyn_jacket_side_02_l_box
   center=<6.5287 -0.1533 -0.2253>
   rotation=<rot 11.2004 95.401 -86.204>
  }]
 dyn_jacket_side_02_r [{
   shapeType=Box
   extentX=2.6723
   extentY=8.4363
   extentZ=13.1087
   shapeName=dyn_jacket_side_02_r_box
   center=<6.4707 -0.0725 -0.1075>
   rotation=<rot 1.04 -82.7918 -80.6902>
  }]
 dyn_sword [{
   shapeType=Capsule
   length=59.8244
   radius=2.1373
   shapeName=dyn_sword_capsule
   center=<23.1499 -0.239 -0.8155>
   rotation=<rot 0 90.2014 -87.5111>
  }]
 pelvis [{
   shapeType=Capsule
   length=3.7254
   radius=10.0262
   shapeName=pelvis_capsule
   center=<0.5034 0.2447 0.0169>
   rotation=<rot 0.2164 -90.0001 0.4869>
  }]
 spine_01 [{
   shapeType=Capsule
   length=2.0104
   radius=10.2558
   shapeName=pelvis_capsule
   center=<0.5034 1.8133 0.011>
   rotation=<rot 0.2164 -90.0001 0.4869>
  }]
 thigh_l [{shapeType=Capsule, length=22.9265, radius=8.149, shapeName=thigh_l_capsule, center=<-9.8482 0 2.961>, rotation=<rot 90 0 0>}]
 thigh_r [{shapeType=Capsule, length=22.7904, radius=8.149, shapeName=thigh_r_capsule, center=<10.1477 0 -2.8737>, rotation=<rot 90 180 0>}]
}
constraints [
 {bone1Name=dyn_jacket_side_01_l, bone2Name=MSTR_jacket_side_l, swing1LimitDegrees=20, swing2LimitDegrees=20, twistLimitDegrees=2}
 {bone1Name=dyn_jacket_front_02_l, bone2Name=dyn_jacket_front_01_l, swing1LimitDegrees=45, swing2LimitDegrees=20, twistLimitDegrees=20}
 {bone1Name=dyn_jacket_front_01_l, bone2Name=MSTR_jacket_front_l, swing1LimitDegrees=20, swing2LimitDegrees=20, twistLimitDegrees=15}
 {bone1Name=dyn_jacket_side_02_l, bone2Name=dyn_jacket_side_01_l, swing1LimitDegrees=45, swing2LimitDegrees=15, twistLimitDegrees=5}
 {
  bone1Name=dyn_jacket_back_02_l
  bone2Name=dyn_jacket_back_01_l
  swing1LimitDegrees=20
  swing2LimitDegrees=20
  twistLimitDegrees=5
  twistMotion=Locked
 }
 {bone1Name=dyn_jacket_back_01_l, bone2Name=MSTR_jacket_back_l, swing1LimitDegrees=15, swing2LimitDegrees=10, twistLimitDegrees=2}
 {
  bone1Name=dyn_jacket_back_02
  bone2Name=dyn_jacket_back_01
  swing1LimitDegrees=20
  swing2LimitDegrees=20
  twistLimitDegrees=5
  twistMotion=Locked
 }
 {bone1Name=dyn_jacket_back_01, bone2Name=MSTR_jacket_back, swing1LimitDegrees=15, swing2LimitDegrees=10, twistLimitDegrees=2}
 {
  bone1Name=dyn_jacket_back_02_r
  bone2Name=dyn_jacket_back_01_r
  swing1LimitDegrees=20
  swing2LimitDegrees=20
  twistLimitDegrees=5
  twistMotion=Locked
 }
 {bone1Name=dyn_jacket_back_01_r, bone2Name=MSTR_jacket_back_r, swing1LimitDegrees=15, swing2LimitDegrees=10, twistLimitDegrees=2}
 {bone1Name=dyn_jacket_side_01_r, bone2Name=MSTR_jacket_side_r, swing1LimitDegrees=10, swing2LimitDegrees=10, twistLimitDegrees=2}
 {bone1Name=dyn_jacket_side_02_r, bone2Name=dyn_jacket_side_01_r, swing1LimitDegrees=20, swing2LimitDegrees=15, twistLimitDegrees=5}
 {bone1Name=dyn_jacket_front_01_r, bone2Name=MSTR_jacket_front_r, swing1LimitDegrees=10, swing2LimitDegrees=10, twistLimitDegrees=5}
 {bone1Name=dyn_jacket_front_02_r, bone2Name=dyn_jacket_front_01_r, swing1LimitDegrees=10, swing2LimitDegrees=10, twistLimitDegrees=5}
 {
  bone1Name=dyn_jacket_side_02_l
  bone2Name=dyn_jacket_front_02_l
  swing1LimitDegrees=45
  swing1Motion=Free
  swing2LimitDegrees=45
  swing2Motion=Free
  twistLimitDegrees=45
 }
 {
  bone1Name=dyn_jacket_side_02_r
  bone2Name=dyn_jacket_front_02_r
  swing1LimitDegrees=45
  swing1Motion=Free
  swing2LimitDegrees=45
  swing2Motion=Free
  twistLimitDegrees=45
  twistMotion=Free
 }
 {
  bone1Name=dyn_jacket_back_02_l
  bone2Name=dyn_jacket_side_02_l
  swing1LimitDegrees=45
  swing1Motion=Free
  swing2LimitDegrees=45
  swing2Motion=Free
  twistLimitDegrees=45
  twistMotion=Free
 }
 {
  bone1Name=dyn_jacket_back_02_r
  bone2Name=dyn_jacket_side_02_r
  swing1LimitDegrees=45
  swing1Motion=Free
  swing2LimitDegrees=45
  swing2Motion=Free
  twistLimitDegrees=45
  twistMotion=Free
 }
 {
  bone1Name=dyn_jacket_back_02
  bone2Name=dyn_jacket_back_02_r
  swing1LimitDegrees=45
  swing1Motion=Free
  swing2LimitDegrees=45
  swing2Motion=Free
  twistLimitDegrees=45
  twistMotion=Free
 }
 {
  bone1Name=dyn_jacket_back_02
  bone2Name=dyn_jacket_back_02_l
  swing1LimitDegrees=45
  swing1Motion=Free
  swing2LimitDegrees=45
  swing2Motion=Free
  twistLimitDegrees=45
  twistMotion=Free
 }
]
