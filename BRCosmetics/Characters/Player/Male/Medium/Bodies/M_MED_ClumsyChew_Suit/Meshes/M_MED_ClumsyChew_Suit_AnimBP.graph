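# Post-process style AnimBlueprint for the ClumsyChew Suit cosmetic body, adding secondary motion to tie, sword, fur, jacket and buttons.
# Parent pose feeds a Control Rig, weighted bone constraints, Anim Dynamics chains and RigidBody, then jacket-button constraints.
# A front-end idle curve blends DefaultSlot animation over belt and jacket branches.
# Events disable sword dynamics while sliding or DBNO and switch rope simulation to world space in the front end.
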AnimBlueprint /BRCosmetics/Characters/Player/Male/Medium/Bodies/M_MED_ClumsyChew_Suit/Meshes/M_MED_ClumsyChew_Suit_AnimBP
target_skeleton=&/M_MED_ClumsyChew_Suit_Skeleton
graphs [
 AnimGraph {nodes [
   Constraint!(
    ComponentPose
    Alpha=1
    ConstraintWeights_0=0.0267
    ConstraintWeights_1:float=0.087
    ConstraintWeights_2:float=0.2388
    ConstraintWeights_3:float=0.4163
    ConstraintWeights_4:float=0.0031
    ConstraintWeights_5:float=0.0103
    ConstraintWeights_6:float=0.2138
    ConstraintWeights_7:float=0.004
   ) -> (Pose) {
    alphaBoolBlend {blendOption=Linear}
    constraintSetup [
     {targetBone {boneName=spine_02}}
     {targetBone {boneName=spine_03}}
     {targetBone {boneName=spine_04}}
     {targetBone {boneName=spine_05}}
     {targetBone {boneName=neck_01}}
     {targetBone {boneName=neck_02}}
     {targetBone {boneName=clavicle_r}}
     {targetBone {boneName=deform_upperarm_up_r}}
    ]
    constraintWeights [1,1,1,1,1,1,1,1]
    boneToModify=~
   }
   n1 = Layered blend per bone(
    BasePose=Use cached pose 'Start Pose'()
    BlendPoses_0=Slot 'DefaultSlot'()
    BlendWeights_0=Get Custom Idle is Playing Float() -> (Custom Idle is Playing Float:double)
   ) {layerSetup [{branchFilters [{blendDepth=-1, boneName=dyb_swordBelt}, {boneName=MSTR_jacket_front_r}, {boneName=MSTR_jacket_side_r}]}]}
   Output Pose(Result=@n1) {layerGroup=None}
   n2 = Control Rig!(Source=Copy Pose From Mesh() {bUseAttachedParent=true}, Alpha, isFrontEnd:bool=Get bIsInFrontEnd()) -> (Pose) {
    alphaBoolBlend {blendOption=Linear}
    controlRigAssetReference {blueprintRigClass=&/M_MED_ClumsyChew_Suit_CtrlRig.M_MED_ClumsyChew_Suit_CtrlRig_C}
   }
   n3 = "Constraint - dyn_handFur_03_l (PPPPPPPP)"!(
    ComponentPose=Local To Component(LocalPose=@n2)
    Alpha=1
    ConstraintWeights_0=0.1507
    ConstraintWeights_1:float=0.0609
    ConstraintWeights_2:float=0.0394
    ConstraintWeights_3:float=0.0056
    ConstraintWeights_4:float=0.0065
    ConstraintWeights_5:float=0.2651
    ConstraintWeights_6:float=0.4718
    ConstraintWeights_7:float=0.0001
   ) -> (Pose) {
    alphaBoolBlend {blendOption=Linear}
    boneToModify {boneName=dyn_handFur_03_l}
    constraintSetup [
     {targetBone {boneName=hand_l}}
     {targetBone {boneName=thumb_01_l}}
     {targetBone {boneName=index_metacarpal_l}}
     {targetBone {boneName=middle_metacarpal_l}}
     {targetBone {boneName=ring_metacarpal_l}}
     {targetBone {boneName=deform_wrist_up_l}}
     {targetBone {boneName=lowerarm_twist_01_l}}
     {targetBone {boneName=dyn_handFur_03_l}}
    ]
    constraintWeights [1,1,1,1,1,1,1,1]
   }
   n4 = "Constraint - dyn_handFur_02_l (PPPPPPP)"!(
    ComponentPose=@n3
    Alpha=1
    ConstraintWeights_0=0.1607
    ConstraintWeights_1:float=0.0114
    ConstraintWeights_2:float=0.0245
    ConstraintWeights_3:float=0.0161
    ConstraintWeights_4:float=0.0242
    ConstraintWeights_5:float=0.3698
    ConstraintWeights_6:float=0.3933
   ) -> (Pose) {
    alphaBoolBlend {blendOption=Linear}
    boneToModify {boneName=dyn_handFur_02_l}
    constraintSetup [
     {targetBone {boneName=hand_l}}
     {targetBone {boneName=thumb_01_l}}
     {targetBone {boneName=index_metacarpal_l}}
     {targetBone {boneName=ring_metacarpal_l}}
     {targetBone {boneName=pinky_metacarpal_l}}
     {targetBone {boneName=deform_wrist_up_l}}
     {targetBone {boneName=lowerarm_twist_01_l}}
    ]
    constraintWeights [1,1,1,1,1,1,1]
   }
   n5 = "Constraint - dyn_handFur_03_l (PPPPPPPP)"(
    ComponentPose=@n4
    Alpha=1
    ConstraintWeights_0=0.0036
    ConstraintWeights_1=0.1963
    ConstraintWeights_2=0.0064
    ConstraintWeights_3=0.0033
    ConstraintWeights_4=0.0145
    ConstraintWeights_5=0.068
    ConstraintWeights_6=0.3135
    ConstraintWeights_7=0.3945
   ) {
    alphaBoolBlend {blendOption=Linear}
    boneToModify {boneName=dyn_handFur_03_l}
    constraintSetup [
     {targetBone {boneName=lowerarm_l}}
     {targetBone {boneName=hand_l}}
     {targetBone {boneName=index_metacarpal_l}}
     {targetBone {boneName=middle_metacarpal_l}}
     {targetBone {boneName=ring_metacarpal_l}}
     {targetBone {boneName=pinky_metacarpal_l}}
     {targetBone {boneName=deform_wrist_up_l}}
     {targetBone {boneName=lowerarm_twist_01_l}}
    ]
    constraintWeights [1,1,1,1,1,1,1,1]
   }
   n6 = "Constraint - dyn_handFur_03_r (PPPPPPPP)"!(
    ComponentPose=@n5
    Alpha=1
    ConstraintWeights_0=0.1505
    ConstraintWeights_1:float=0.0609
    ConstraintWeights_2:float=0.0393
    ConstraintWeights_3:float=0.0055
    ConstraintWeights_4:float=0.0064
    ConstraintWeights_5:float=0.2642
    ConstraintWeights_6:float=0.4724
    ConstraintWeights_7:float=0.0008
   ) -> (Pose) {
    alphaBoolBlend {blendOption=Linear}
    boneToModify {boneName=dyn_handFur_03_r}
    constraintSetup [
     {targetBone {boneName=hand_r}}
     {targetBone {boneName=thumb_01_r}}
     {targetBone {boneName=index_metacarpal_r}}
     {targetBone {boneName=middle_metacarpal_r}}
     {targetBone {boneName=ring_metacarpal_r}}
     {targetBone {boneName=deform_wrist_up_r}}
     {targetBone {boneName=lowerarm_twist_01_r}}
     {targetBone {boneName=dyn_handFur_03_r}}
    ]
    constraintWeights [1,1,1,1,1,1,1,1]
   }
   n7 = "Constraint - dyn_handFur_02_r (PPPPPPP)"!(
    ComponentPose=@n6
    Alpha=1
    ConstraintWeights_0=0.1617
    ConstraintWeights_1:float=0.0113
    ConstraintWeights_2:float=0.0246
    ConstraintWeights_3:float=0.0163
    ConstraintWeights_4:float=0.0248
    ConstraintWeights_5:float=0.3683
    ConstraintWeights_6:float=0.393
   ) -> (Pose) {
    alphaBoolBlend {blendOption=Linear}
    boneToModify {boneName=dyn_handFur_02_r}
    constraintSetup [
     {targetBone {boneName=hand_r}}
     {targetBone {boneName=thumb_01_r}}
     {targetBone {boneName=index_metacarpal_r}}
     {targetBone {boneName=ring_metacarpal_r}}
     {targetBone {boneName=pinky_metacarpal_r}}
     {targetBone {boneName=deform_wrist_up_r}}
     {targetBone {boneName=lowerarm_twist_01_r}}
    ]
    constraintWeights [1,1,1,1,1,1,1]
   }
   n8 = "Constraint - dyn_handFur_01_r (PPPPPPPP)"!(
    ComponentPose=@n7
    Alpha=1
    ConstraintWeights_0=0.0038
    ConstraintWeights_1:float=0.1917
    ConstraintWeights_2:float=0.0063
    ConstraintWeights_3:float=0.0029
    ConstraintWeights_4:float=0.014
    ConstraintWeights_5:float=0.0656
    ConstraintWeights_6:float=0.3133
    ConstraintWeights_7:float=0.4023
   ) -> (Pose) {
    alphaBoolBlend {blendOption=Linear}
    boneToModify {boneName=dyn_handFur_01_r}
    constraintSetup [
     {targetBone {boneName=lowerarm_r}}
     {targetBone {boneName=hand_r}}
     {targetBone {boneName=index_metacarpal_r}}
     {targetBone {boneName=middle_metacarpal_r}}
     {targetBone {boneName=ring_metacarpal_r}}
     {targetBone {boneName=pinky_metacarpal_r}}
     {targetBone {boneName=deform_wrist_up_r}}
     {targetBone {boneName=lowerarm_twist_01_r}}
    ]
    constraintWeights [1,1,1,1,1,1,1,1]
   }
   n9 = "Constraint - def_armButton_01_l (PPPPPP)"!(
    ComponentPose=@n8
    Alpha=1
    ConstraintWeights_0=0.0677
    ConstraintWeights_1:float=0.0365
    ConstraintWeights_2:float=0.0008
    ConstraintWeights_3:float=0.0195
    ConstraintWeights_4:float=0.292
    ConstraintWeights_5:float=0.5836
   ) -> (Pose) {
    alphaBoolBlend {blendOption=Linear}
    boneToModify {boneName=def_armButton_01_l}
    constraintSetup [
     {targetBone {boneName=lowerarm_l}}
     {targetBone {boneName=hand_l}}
     {targetBone {boneName=thumb_01_l}}
     {targetBone {boneName=pinky_metacarpal_l}}
     {targetBone {boneName=lowerarm_twist_02_l}}
     {targetBone {boneName=lowerarm_twist_01_l}}
    ]
    constraintWeights [1,1,1,1,1,1]
   }
   n10 = "Constraint - def_armButton_02_l (PPPPPPP)"!(
    ComponentPose=@n9
    Alpha=1
    ConstraintWeights_0=0.0354
    ConstraintWeights_1:float=0.0553
    ConstraintWeights_2:float=0.0009
    ConstraintWeights_3:float=0.0289
    ConstraintWeights_4:float=0.0137
    ConstraintWeights_5:float=0.1933
    ConstraintWeights_6:float=0.6725
   ) -> (Pose) {
    alphaBoolBlend {blendOption=Linear}
    boneToModify {boneName=def_armButton_02_l}
    constraintSetup [
     {targetBone {boneName=lowerarm_l}}
     {targetBone {boneName=hand_l}}
     {targetBone {boneName=thumb_01_l}}
     {targetBone {boneName=pinky_metacarpal_l}}
     {targetBone {boneName=deform_wrist_up_l}}
     {targetBone {boneName=lowerarm_twist_02_l}}
     {targetBone {boneName=lowerarm_twist_01_l}}
    ]
    constraintWeights [1,1,1,1,1,1,1]
   }
   n11 = "Constraint - def_armButton_01_r (PPPPPP)"!(
    ComponentPose=@n10
    Alpha=1
    ConstraintWeights_0=0.0677
    ConstraintWeights_1:float=0.0365
    ConstraintWeights_2:float=0.0008
    ConstraintWeights_3:float=0.0195
    ConstraintWeights_4:float=0.292
    ConstraintWeights_5:float=0.5836
   ) -> (Pose) {
    alphaBoolBlend {blendOption=Linear}
    boneToModify {boneName=def_armButton_01_r}
    constraintSetup [
     {targetBone {boneName=lowerarm_r}}
     {targetBone {boneName=hand_r}}
     {targetBone {boneName=thumb_01_r}}
     {targetBone {boneName=pinky_metacarpal_r}}
     {targetBone {boneName=lowerarm_twist_02_r}}
     {targetBone {boneName=lowerarm_twist_01_r}}
    ]
    constraintWeights [1,1,1,1,1,1]
   }
   n12 = "Constraint - def_armButton_02_r (PPPPPPP)"!(
    ComponentPose=@n11
    Alpha=1
    ConstraintWeights_0=0.0354
    ConstraintWeights_1:float=0.0553
    ConstraintWeights_2:float=0.0009
    ConstraintWeights_3:float=0.0289
    ConstraintWeights_4:float=0.0137
    ConstraintWeights_5:float=0.1933
    ConstraintWeights_6:float=0.6725
   ) -> (Pose) {
    alphaBoolBlend {blendOption=Linear}
    boneToModify {boneName=def_armButton_02_r}
    constraintSetup [
     {targetBone {boneName=lowerarm_r}}
     {targetBone {boneName=hand_r}}
     {targetBone {boneName=thumb_01_r}}
     {targetBone {boneName=pinky_metacarpal_r}}
     {targetBone {boneName=deform_wrist_up_r}}
     {targetBone {boneName=lowerarm_twist_02_r}}
     {targetBone {boneName=lowerarm_twist_01_r}}
    ]
    constraintWeights [1,1,1,1,1,1,1]
   }
   n13 = "Constraint - MSTR_tie (PPPPPPPP)"!(
    ComponentPose=@n12
    Alpha=1
    ConstraintWeights_0=0.001
    ConstraintWeights_1:float=0.022
    ConstraintWeights_2:float=0.1068
    ConstraintWeights_3:float=0.2384
    ConstraintWeights_4:float=0.3741
    ConstraintWeights_5:float=0.2545
    ConstraintWeights_6:float=0.0016
    ConstraintWeights_7:float=0.0017
   ) -> (Pose) {
    alphaBoolBlend {blendOption=Linear}
    boneToModify {boneName=MSTR_tie}
    constraintSetup [
     {targetBone {boneName=pelvis}}
     {targetBone {boneName=spine_01}}
     {targetBone {boneName=spine_02}}
     {targetBone {boneName=spine_03}}
     {targetBone {boneName=spine_04}}
     {targetBone {boneName=spine_05}}
     {targetBone {boneName=clavicle_l}}
     {targetBone {boneName=clavicle_r}}
    ]
    constraintWeights [1,1,1,1,1,1,1,1]
   }
   n14 = Anim Dynamics - Bone: dyn_tie01!(ComponentPose=@n13, Alpha=1, GravityOverride:Vector="0,0,-980", ExternalForce="0,0,0") -> (Pose) {
    angularDampingOverride=0.99
    angularSpringConstant=200
    bAngularSpring=true
    bOverrideAngularDamping=true
    bUseGravityOverride=true
    bUseSphericalLimits=true
    simulationSpace=RootRelative
    alphaBoolBlend {blendOption=Linear}
    boundBone {boneName=dyn_tie01}
    componentAppliedLinearAccClamp=<3000 3000 3000>
    componentLinearAccScale=<1 1 1>
    componentLinearVelScale=<1 1 1>
    gravityOverride {z=-980}
    physicsBodyDefinitions [{
      collisionType=CustomSphere
      sphereCollisionRadius=2
      boundBone {boneName=dyn_tie01}
      boxExtents=<15 15 15>
      constraintSetup {angularLimitsMax=<0 5 0>, angularLimitsMin=<0 -5 -60>, angularTarget=<1 0 0>}
      localJointOffset=<1 0 0>
     }]
    sphericalLimits [{limitRadius=10, drivingBone {boneName=spine_04}, sphereLocalOffset=<1 0 0>}]
   }
   n15 = Anim Dynamics - Bone: dyn_tie02!(ComponentPose=@n14, Alpha=1, GravityOverride:Vector="0,0,-980", ExternalForce="0,0,0") -> (Pose) {
    angularDampingOverride=0.99
    angularSpringConstant=200
    bAngularSpring=true
    bOverrideAngularDamping=true
    bUseGravityOverride=true
    bUseSphericalLimits=true
    simulationSpace=RootRelative
    alphaBoolBlend {blendOption=Linear}
    boundBone {boneName=dyn_tie02}
    componentAppliedLinearAccClamp=<3000 3000 3000>
    componentLinearAccScale=<1 1 1>
    componentLinearVelScale=<1 1 1>
    gravityOverride {z=-980}
    physicsBodyDefinitions [{
      collisionType=CustomSphere
      sphereCollisionRadius=2
      boundBone {boneName=dyn_tie02}
      boxExtents=<10 10 10>
      constraintSetup {angularLimitsMax=<0 10 0>, angularLimitsMin=<0 -10 -35>, angularTarget=<1 0 0>}
      localJointOffset=<1 0 0>
     }]
    sphericalLimits [
     {limitRadius=9, drivingBone {boneName=spine_04}, sphereLocalOffset=<0 0 0>}
     {limitRadius=9.5, drivingBone {boneName=spine_03}, sphereLocalOffset=<1 0 0>}
    ]
   }
   n16 = Anim Dynamics - Bone: dyn_tie03!(ComponentPose=@n15, Alpha=1, GravityOverride:Vector="0,0,-980", ExternalForce="0,0,0") -> (Pose) {
    angularDampingOverride=0.99
    angularSpringConstant=200
    bAngularSpring=true
    bOverrideAngularDamping=true
    bUseGravityOverride=true
    bUsePlanarLimit=false
    bUseSphericalLimits=true
    simulationSpace=RootRelative
    alphaBoolBlend {blendOption=Linear}
    boundBone {boneName=dyn_tie03}
    componentAppliedLinearAccClamp=<3000 3000 3000>
    componentLinearAccScale=<1 1 1>
    componentLinearVelScale=<1 1 1>
    gravityOverride {z=-980}
    physicsBodyDefinitions [{
      collisionType=CustomSphere
      sphereCollisionRadius=2
      boundBone {boneName=dyn_tie03}
      boxExtents=<10 10 10>
      constraintSetup {angularLimitsMax=<0 10 0>, angularLimitsMin=<0 -10 -15>, angularTarget=<1 0 0>}
      localJointOffset=<1 0 0>
     }]
    sphericalLimits [{limitRadius=9, drivingBone {boneName=spine_03}, sphereLocalOffset=<1 0 0>}]
   }
   n17 = Anim Dynamics - Bone: dyb_swordBelt!(ComponentPose=@n16, Alpha=1, GravityOverride:Vector="0,0,-980", ExternalForce="0,0,0") -> (Pose) {
    angularDampingOverride=0.99
    angularSpringConstant=100
    bAngularSpring=true
    bOverrideAngularDamping=true
    bUseGravityOverride=true
    simulationSpace=RootRelative
    alphaBoolBlend {blendOption=Linear}
    boundBone {boneName=dyb_swordBelt}
    componentAppliedLinearAccClamp=<3000 3000 3000>
    componentLinearAccScale=<1 1 1>
    componentLinearVelScale=<1 1 1>
    gravityOverride {z=-980}
    physicsBodyDefinitions [{
      boundBone {boneName=dyb_swordBelt}
      boxExtents=<25 25 18>
      constraintSetup {angularLimitsMax=<0 5 3>, angularLimitsMin=<0 -5 -15>, angularTarget=<1 0 0>}
      localJointOffset=<0.5 0 0>
     }]
   }
   n18 = Anim Dynamics - Bone: dyn_sword!(
    ComponentPose=@n17
    Alpha=Get isSwordAnimDynamics() -> (isSwordAnimDynamics:double)
    GravityOverride:Vector="0,0,-980"
    ExternalForce="0,0,0"
   ) -> (Pose) {
    angularDampingOverride=0.99
    angularSpringConstant=50
    bAngularSpring=true
    bOverrideAngularDamping=true
    bUseGravityOverride=true
    simulationSpace=RootRelative
    alphaBoolBlend {blendOption=Linear}
    boundBone {boneName=dyn_sword}
    componentAppliedLinearAccClamp=<3000 3000 3000>
    componentLinearAccScale=<0.5 0.5 0.5>
    componentLinearVelScale=<0.5 0.5 0.5>
    gravityOverride {z=-980}
    physicsBodyDefinitions [{
      boundBone {boneName=dyn_sword}
      boxExtents=<30 15 15>
      constraintSetup {angularLimitsMax=<0 15 0>, angularLimitsMin=<0 -15 -15>, angularTarget=<1 0 0>}
      localJointOffset=<2 2 0>
     }]
   }
   n19 = Get Simulation Space() -> (Simulation Space:AnimPhysSimSpaceType Enum)
   n20 = Anim Dynamics - Bone: dyn_sword_rope_01!(SimulationSpace:AnimPhysSimSpaceType Enum=@n19, ComponentPose=@n18, Alpha=1, GravityOverride:Vector="0,0,-980", ExternalForce="0,0,0") -> (Pose) {
    angularDampingOverride=0.99
    angularSpringConstant=300
    bAngularSpring=true
    bOverrideAngularDamping=true
    bUseGravityOverride=true
    simulationSpace=RootRelative
    alphaBoolBlend {blendOption=Linear}
    boundBone {boneName=dyn_sword_rope_01}
    componentAppliedLinearAccClamp=<3000 3000 3000>
    componentLinearAccScale=<2 2 2>
    componentLinearVelScale=<2 2 2>
    gravityOverride {z=-980}
    physicsBodyDefinitions [{
      boundBone {boneName=dyn_sword_rope_01}
      boxExtents=<11 11 11>
      constraintSetup {angularLimitsMax=<0 20 20>, angularLimitsMin=<0 -20 -20>, angularTarget=<1 0 0>}
      localJointOffset=<1 0 0>
     }]
   }
   n21 = Anim Dynamics - Bone: dyn_sword_rope_02!(SimulationSpace:AnimPhysSimSpaceType Enum=@n19, ComponentPose=@n20, Alpha=1, GravityOverride:Vector="0,0,-980", ExternalForce="0,0,0") -> (Pose) {
    angularDampingOverride=0.99
    angularSpringConstant=300
    bAngularSpring=true
    bOverrideAngularDamping=true
    bUseGravityOverride=true
    simulationSpace=RootRelative
    alphaBoolBlend {blendOption=Linear}
    boundBone {boneName=dyn_sword_rope_02}
    componentAppliedLinearAccClamp=<3000 3000 3000>
    componentLinearAccScale=<2 2 2>
    componentLinearVelScale=<2 2 2>
    gravityOverride {z=-980}
    physicsBodyDefinitions [{
      boundBone {boneName=dyn_sword_rope_02}
      boxExtents=<11 11 11>
      constraintSetup {angularLimitsMax=<0 20 20>, angularLimitsMin=<0 -20 -20>, angularTarget=<1 0 0>}
      localJointOffset=<1 0 0>
     }]
   }
   n22 = Anim Dynamics - Bone: dyn_sword_rope_03!(SimulationSpace:AnimPhysSimSpaceType Enum=@n19, ComponentPose=@n21, Alpha=1, GravityOverride:Vector="0,0,-980", ExternalForce="0,0,0") -> (Pose) {
    angularDampingOverride=0.99
    angularSpringConstant=300
    bAngularSpring=true
    bOverrideAngularDamping=true
    bUseGravityOverride=true
    simulationSpace=RootRelative
    alphaBoolBlend {blendOption=Linear}
    boundBone {boneName=dyn_sword_rope_03}
    componentAppliedLinearAccClamp=<3000 3000 3000>
    componentLinearAccScale=<2 2 2>
    componentLinearVelScale=<2 2 2>
    gravityOverride {z=-980}
    physicsBodyDefinitions [{
      boundBone {boneName=dyn_sword_rope_03}
      boxExtents=<11 11 11>
      constraintSetup {angularLimitsMax=<0 20 20>, angularLimitsMin=<0 -20 -20>, angularTarget=<1 0 0>}
      localJointOffset=<1 0 0>
     }]
   }
   n23 = Anim Dynamics - Bone: dyn_handFur_01_l!(ComponentPose=@n22, Alpha=1, GravityOverride:Vector="-980,0,0", ExternalForce="0,0,0") -> (Pose) {
    angularDampingOverride=0.999
    angularSpringConstant=150
    bAngularSpring=true
    bGravityOverrideInSimSpace=true
    bOverrideAngularDamping=true
    bUseGravityOverride=true
    simulationSpace=BoneRelative
    alphaBoolBlend {blendOption=Linear}
    boundBone {boneName=dyn_handFur_01_l}
    componentAppliedLinearAccClamp=<3000 3000 3000>
    componentLinearAccScale=<1 1 1>
    componentLinearVelScale=<1 1 1>
    gravityOverride {z=-980}
    physicsBodyDefinitions [{
      boundBone {boneName=dyn_handFur_01_l}
      boxExtents=<10 10 10>
      constraintSetup {angularLimitsMax=<0 45 10>, angularLimitsMin=<0 -45 -45>, angularTarget=<1 0 0>}
      localJointOffset=<-3 0 0>
     }]
    relativeSpaceBone {boneName=lowerarm_l}
   }
   n24 = Anim Dynamics - Bone: dyn_handFur_02_l!(ComponentPose=@n23, Alpha=1, GravityOverride:Vector="-980,0,0", ExternalForce="0,0,0") -> (Pose) {
    angularDampingOverride=0.999
    angularSpringConstant=150
    bAngularSpring=true
    bGravityOverrideInSimSpace=true
    bOverrideAngularDamping=true
    bUseGravityOverride=true
    simulationSpace=BoneRelative
    alphaBoolBlend {blendOption=Linear}
    boundBone {boneName=dyn_handFur_02_l}
    componentAppliedLinearAccClamp=<3000 3000 3000>
    componentLinearAccScale=<1 1 1>
    componentLinearVelScale=<1 1 1>
    gravityOverride {z=-980}
    physicsBodyDefinitions [{
      boundBone {boneName=dyn_handFur_02_l}
      boxExtents=<10 10 10>
      constraintSetup {angularLimitsMax=<0 45 10>, angularLimitsMin=<0 -45 -45>, angularTarget=<1 0 0>}
      localJointOffset=<-3 0 0>
     }]
    relativeSpaceBone {boneName=lowerarm_l}
   }
   n25 = Anim Dynamics - Bone: dyn_handFur_03_l!(ComponentPose=@n24, Alpha=1, GravityOverride:Vector="-980,0,0", ExternalForce="0,0,0") -> (Pose) {
    angularDampingOverride=0.999
    angularSpringConstant=150
    bAngularSpring=true
    bGravityOverrideInSimSpace=true
    bOverrideAngularDamping=true
    bUseGravityOverride=true
    simulationSpace=BoneRelative
    alphaBoolBlend {blendOption=Linear}
    boundBone {boneName=dyn_handFur_03_l}
    componentAppliedLinearAccClamp=<3000 3000 3000>
    componentLinearAccScale=<1 1 1>
    componentLinearVelScale=<1 1 1>
    gravityOverride {z=-980}
    physicsBodyDefinitions [{
      boundBone {boneName=dyn_handFur_03_l}
      boxExtents=<10 10 10>
      constraintSetup {angularLimitsMax=<0 45 10>, angularLimitsMin=<0 -45 -45>, angularTarget=<1 0 0>}
      localJointOffset=<-3 0 0>
     }]
    relativeSpaceBone {boneName=lowerarm_l}
   }
   n26 = Anim Dynamics - Bone: dyn_handFur_01_r!(ComponentPose=@n25, Alpha=1, GravityOverride:Vector="-980,0,0", ExternalForce="0,0,0") -> (Pose) {
    angularDampingOverride=0.999
    angularSpringConstant=150
    bAngularSpring=true
    bGravityOverrideInSimSpace=true
    bOverrideAngularDamping=true
    bUseGravityOverride=true
    simulationSpace=BoneRelative
    alphaBoolBlend {blendOption=Linear}
    boundBone {boneName=dyn_handFur_01_r}
    componentAppliedLinearAccClamp=<3000 3000 3000>
    componentLinearAccScale=<1 1 1>
    componentLinearVelScale=<1 1 1>
    gravityOverride {z=-980}
    physicsBodyDefinitions [{
      boundBone {boneName=dyn_handFur_01_r}
      boxExtents=<10 10 10>
      constraintSetup {angularLimitsMax=<0 45 10>, angularLimitsMin=<0 -45 -45>, angularTarget=<1 0 0>}
      localJointOffset=<3 0 0>
     }]
    relativeSpaceBone {boneName=lowerarm_r}
   }
   n27 = Anim Dynamics - Bone: dyn_handFur_02_r!(ComponentPose=@n26, Alpha=1, GravityOverride:Vector="-980,0,0", ExternalForce="0,0,0") -> (Pose) {
    angularDampingOverride=0.999
    angularSpringConstant=150
    bAngularSpring=true
    bGravityOverrideInSimSpace=true
    bOverrideAngularDamping=true
    bUseGravityOverride=true
    simulationSpace=BoneRelative
    alphaBoolBlend {blendOption=Linear}
    boundBone {boneName=dyn_handFur_02_r}
    componentAppliedLinearAccClamp=<3000 3000 3000>
    componentLinearAccScale=<1 1 1>
    componentLinearVelScale=<1 1 1>
    gravityOverride {z=-980}
    physicsBodyDefinitions [{
      boundBone {boneName=dyn_handFur_02_r}
      boxExtents=<10 10 10>
      constraintSetup {angularLimitsMax=<0 45 10>, angularLimitsMin=<0 -45 -45>, angularTarget=<1 0 0>}
      localJointOffset=<3 0 0>
     }]
    relativeSpaceBone {boneName=lowerarm_r}
   }
   n28 = Anim Dynamics - Bone: dyn_handFur_03_r!(ComponentPose=@n27, Alpha=1, GravityOverride:Vector="-980,0,0", ExternalForce="0,0,0") -> (Pose) {
    angularDampingOverride=0.999
    angularSpringConstant=150
    bAngularSpring=true
    bGravityOverrideInSimSpace=true
    bOverrideAngularDamping=true
    bUseGravityOverride=true
    simulationSpace=BoneRelative
    alphaBoolBlend {blendOption=Linear}
    boundBone {boneName=dyn_handFur_03_r}
    componentAppliedLinearAccClamp=<3000 3000 3000>
    componentLinearAccScale=<1 1 1>
    componentLinearVelScale=<1 1 1>
    gravityOverride {z=-980}
    physicsBodyDefinitions [{
      boundBone {boneName=dyn_handFur_03_r}
      boxExtents=<10 10 10>
      constraintSetup {angularLimitsMax=<0 45 10>, angularLimitsMin=<0 -45 -45>, angularTarget=<1 0 0>}
      localJointOffset=<3 0 0>
     }]
    relativeSpaceBone {boneName=lowerarm_r}
   }
   n29 = RigidBody!(ComponentPose=@n28, Alpha, OverrideWorldGravity:Vector="0,0,0", ExternalForce) -> (Pose) {
    bClampLinearTranslationLimitToRefPose=true
    bOverrideWorldGravity=true
    overridePhysicsAsset=&/M_MED_ClumsyChew_Suit_Physics
    simulationSpace=BaseBoneSpace
    alphaBoolBlend {blendOption=Linear}
    baseBoneRef {boneName=pelvis}
    simSpaceSettings {
     dampingAlpha=0
     maxAngularAcceleration=30
     maxAngularVelocity=6
     maxLinearAcceleration=500
     maxLinearVelocity=150
     velocityScaleZ=0.7
     worldAlpha=1
     externalLinearDragV=<0.5 0.5 0.5>
    }
   }
   n30 = "Constraint - def_button_02 (PPPPPPPP)"!(
    ComponentPose=@n29
    Alpha=1
    ConstraintWeights_0=0.1351
    ConstraintWeights_1:float=0.115
    ConstraintWeights_2:float=0.1751
    ConstraintWeights_3:float=0.1629
    ConstraintWeights_4:float=0.0231
    ConstraintWeights_5:float=0.0145
    ConstraintWeights_6:float=0.3648
    ConstraintWeights_7:float=0.0096
   ) -> (Pose) {
    alphaBoolBlend {blendOption=Linear}
    boneToModify {boneName=def_button_02}
    constraintSetup [
     {targetBone {boneName=pelvis}}
     {targetBone {boneName=spine_01}}
     {targetBone {boneName=spine_02}}
     {targetBone {boneName=spine_03}}
     {targetBone {boneName=spine_04}}
     {targetBone {boneName=dyn_jacket_side_01_l}}
     {targetBone {boneName=dyn_jacket_front_01_l}}
     {targetBone {boneName=dyn_jacket_front_02_l}}
    ]
    constraintWeights [1,1,1,1,1,1,1,1]
   }
   n31 = "Constraint - def_button_01 (PPPPPPPP)"!(
    ComponentPose=@n30
    Alpha=1
    ConstraintWeights_0=0.1192
    ConstraintWeights_1:float=0.1479
    ConstraintWeights_2:float=0.2788
    ConstraintWeights_3:float=0.3051
    ConstraintWeights_4:float=0.0784
    ConstraintWeights_5:float=0
    ConstraintWeights_6:float=0
    ConstraintWeights_7:float=0.0706
   ) -> (Pose) {
    alphaBoolBlend {blendOption=Linear}
    boneToModify {boneName=def_button_01}
    constraintSetup [
     {targetBone {boneName=pelvis}}
     {targetBone {boneName=spine_01}}
     {targetBone {boneName=spine_02}}
     {targetBone {boneName=spine_03}}
     {targetBone {boneName=spine_04}}
     {targetBone {boneName=def_button_02}}
     {targetBone {boneName=dyn_jacket_front_01_r}}
     {targetBone {boneName=dyn_jacket_front_01_l}}
    ]
    constraintWeights [1,1,1,1,1,1,1,1]
   }
   Save cached pose 'Start Pose'(Pose=Component To Local(ComponentPose=@n31)) {cacheName=Start Pose}
  ]}
 EventGraph {nodes [
   Event BlueprintUpdateAnimation()
   > n1 = Sequence()
   > n2 = Gravity Override State Function DEPRECATED(
    CharacterPartAnimInstance=Self-Reference()
    GameTimeInSeconds=GetGameTimeInSeconds()
    GravityOverrideParams=Get Gravity Override Params()
    RBAN Params=Get RBAN Params()
   )
   > Set Gravity Override(Gravity Override=@n2)
   n3 = OR Boolean(A=Get bIsSliding(), B=Get bIsDBNO())
   > Set isSwordAnimDynamics(isSwordAnimDynamics:double="To Float (Boolean)"(InBool=NOT Boolean(A=@n3))) -> (Output_Get:double)
   > Set GravityOverrideAnimDynamics(GravityOverrideAnimDynamics:Vector=Select(Option 0:Vector=Get Gravity Override(), Option 1:Vector="0,0,-980", Index=Get bIsDBNO()) -> (ReturnValue:Vector)) -> (Output_Get:Vector)
   Branch(execute=@n1.then_1, Condition=Get bIsInFrontEnd())
   > Set Custom Idle is Playing Float(Custom Idle is Playing Float:double=GetCurveValue(self=GetOwnerFrontEndAnimInstance(), CurveName=Frontend_Idle_Bypass)) -> (Output_Get:double)
   K2Node_Event_1 = CustomEvent!() -> (
    OutputDelegate
    then
    CustomBaseMesh:Skeletal Mesh Component Object Reference
    CustomParts:Array of Custom Character Part Object References
    CustomMeshParts:Array of Skeletal Mesh Component Budgeted Object References
   ) {is=title K2Node_CustomEvent}
   > n4 = Branch(Condition=PickaxeItemDefinitionHasAnyMetaTag(Tags=Make Array("[0]"=(TagName="Cosmetics.Metatags")), bRequiresEquipped=false))
   > "Set ClumsyChew  pickaxe"("ClumsyChew  pickaxe":bool=true) -> (Output_Get:bool)
   "Set ClumsyChew  pickaxe"(execute=@n4.else, "ClumsyChew  pickaxe"=false)
   n5 = Event BlueprintInitializeAnimation()
   Set Simulation Space(
    execute=[@n5.then, @n6.then]
    Simulation Space:AnimPhysSimSpaceType Enum=Select(Option 0:AnimPhysSimSpaceType Enum, Option 1:AnimPhysSimSpaceType Enum=World, Index=Get bIsInFrontEnd()) -> (ReturnValue:AnimPhysSimSpaceType Enum)
   ) -> (Output_Get:AnimPhysSimSpaceType Enum)
   n6 = Event OnCharacterCustomizationLoaded()
  ]}
]
variables [
 "ClumsyChew  pickaxe" {type=boolean}
 Custom Idle is Playing Float {type=number}
 Gravity Override {type=Vector}
 Gravity Override Params {
  type=GravityOverrideParamsStruct
  default {skydive_FloatingDown {gravityOverride=<0 -1000 0>, windAmplitude=<200 -500 150>, windOffset=<0 -600 200>}}
 }
 GravityOverrideAnimDynamics {type=Vector}
 isSwordAnimDynamics {type=number, default=1}
 RBAN Params {
  type=RBANParamsStruct
  default {
   dBNO {simSpaceSettings {
     maxAngularAcceleration=1000
     maxAngularVelocity=3
     maxLinearAcceleration=300
     maxLinearVelocity=700
     velocityScaleZ=0
     externalLinearDragV=<0 0.5 0>
    }}
   dBNO_Carried {simSpaceSettings {
     maxAngularAcceleration=1000
     maxAngularVelocity=3
     maxLinearAcceleration=300
     maxLinearVelocity=700
     velocityScaleZ=0
     externalLinearDragV=<0 0.5 0>
    }}
   emoteOrMelee {simSpaceSettings {
     maxAngularAcceleration=1000
     maxAngularVelocity=3
     maxLinearAcceleration=300
     maxLinearVelocity=700
     velocityScaleZ=0
     externalLinearDragV=<0 0.5 0>
    }}
   front_End {simSpaceSettings {
     maxAngularAcceleration=1000
     maxAngularVelocity=3
     maxLinearAcceleration=300
     maxLinearVelocity=700
     velocityScaleZ=0
     externalLinearDragV=<0 0.5 0>
    }}
   onGround_Crouch_Moving {simSpaceSettings {
     maxAngularAcceleration=1000
     maxAngularVelocity=3
     maxLinearAcceleration=300
     maxLinearVelocity=700
     velocityScaleZ=0
     externalLinearDragV=<0 0.5 0>
    }}
   onGround_Falling {simSpaceSettings {
     maxAngularAcceleration=1000
     maxAngularVelocity=3
     maxLinearAcceleration=300
     maxLinearVelocity=700
     velocityScaleZ=0
     externalLinearDragV=<0 0.5 0>
    }}
   onGround_Moving {simSpaceSettings {
     maxAngularAcceleration=1000
     maxAngularVelocity=3
     maxLinearAcceleration=300
     maxLinearVelocity=700
     velocityScaleZ=0
     externalLinearDragV=<0 0.5 0>
    }}
   onGround_Standing {simSpaceSettings {
     maxAngularAcceleration=1000
     maxAngularVelocity=3
     maxLinearAcceleration=300
     maxLinearVelocity=700
     velocityScaleZ=0
     externalLinearDragV=<0 0.5 0>
    }}
   rail_Grinding {simSpaceSettings {
     maxAngularAcceleration=1000
     maxAngularVelocity=3
     maxLinearAcceleration=300
     maxLinearVelocity=700
     velocityScaleZ=0
     worldAlpha=0.1
     externalLinearDragV=<0 0.5 0>
    }}
   riding_Moving {simSpaceSettings {
     maxAngularAcceleration=1000
     maxAngularVelocity=3
     maxLinearAcceleration=300
     maxLinearVelocity=700
     velocityScaleZ=0
     externalLinearDragV=<0 0.5 0>
    }}
   riding_Moving_Targeting {simSpaceSettings {
     maxAngularAcceleration=1000
     maxAngularVelocity=3
     maxLinearAcceleration=300
     maxLinearVelocity=700
     velocityScaleZ=0
     externalLinearDragV=<0 0.5 0>
    }}
   riding_Standing {simSpaceSettings {
     maxAngularAcceleration=1000
     maxAngularVelocity=3
     maxLinearAcceleration=300
     maxLinearVelocity=700
     velocityScaleZ=0
     externalLinearDragV=<0 0.5 0>
    }}
   skydive_DivingDown {simSpaceSettings {
     maxAngularAcceleration=1000
     maxAngularVelocity=3
     maxLinearAcceleration=300
     maxLinearVelocity=700
     velocityScaleZ=0
     externalLinearDragV=<0 0.5 0>
    }}
   skydive_FloatingDown {
    componentLinearAccScale=<0 0.5 0>
    componentLinearVelScale=<0 0.5 0>
    simSpaceSettings {
     maxAngularAcceleration=1000
     maxAngularVelocity=3
     maxLinearAcceleration=300
     maxLinearVelocity=700
     velocityScaleZ=0
     externalLinearDragV=<0 0.5 0>
    }
   }
   skydive_GoingUp {simSpaceSettings {
     maxAngularAcceleration=1000
     maxAngularVelocity=3
     maxLinearAcceleration=300
     maxLinearVelocity=700
     velocityScaleZ=0
     externalLinearDragV=<0 0.5 0>
    }}
   skydive_Parachuting {
    componentLinearAccScale=<0 0.1 0.3>
    componentLinearVelScale=<0 0.1 0.3>
    simSpaceSettings {
     maxAngularAcceleration=1000
     maxAngularVelocity=3
     maxLinearAcceleration=300
     maxLinearVelocity=700
     velocityScaleZ=0
     externalLinearDragV=<0 0.5 0>
    }
   }
   skydive_Surfing {simSpaceSettings {
     maxAngularAcceleration=1000
     maxAngularVelocity=3
     maxLinearAcceleration=300
     maxLinearVelocity=700
     velocityScaleZ=0
     externalLinearDragV=<0 0.5 0>
    }}
   swimming {simSpaceSettings {
     maxAngularAcceleration=1000
     maxAngularVelocity=3
     maxLinearAcceleration=300
     maxLinearVelocity=700
     velocityScaleZ=0
     externalLinearDragV=<0 0.5 0>
    }}
   tactical_Sprint {simSpaceSettings {
     maxAngularAcceleration=1000
     maxAngularVelocity=3
     maxLinearAcceleration=300
     maxLinearVelocity=700
     velocityScaleZ=0
     externalLinearDragV=<0 0.5 0>
    }}
   ziplining {simSpaceSettings {
     maxAngularAcceleration=1000
     maxAngularVelocity=3
     maxLinearAcceleration=300
     maxLinearVelocity=700
     velocityScaleZ=0
     externalLinearDragV=<0 0.5 0>
    }}
  }
 }
 Simulation Space {type=AnimPhysSimSpaceType, default=RootRelative}
]
